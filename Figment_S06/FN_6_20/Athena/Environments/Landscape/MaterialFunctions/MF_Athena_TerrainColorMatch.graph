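# Terrain color match material function that tints grass to follow a baked Athena terrain color texture.
# World XY is offset and scaled by TerrainSize, computed per vertex through a VertexInterpolator, to sample the combined color map with a mip bias.
# The sample is multiplied by ColormapDivisor from MPC_Landscape, then by NeutralGrassColor.
# Outputs give the tinted result, the neutral color alone (GrassColor) and the untinted map (GrassNoColor).
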MaterialFunction /Figment_S06/FN_6_20/Athena/Environments/Landscape/MaterialFunctions/MF_Athena_TerrainColorMatch
graph {nodes [
  n1 = ScalarParameter() {defaultValue=306200, group=TerrainColorMatch, parameterName=TerrainSize}
  n2 = Divide(A=@n1)
  n3 = Subtract(A=ComponentMask(None=WorldPosition().XYZ) {g=true, r=true}, B=AppendVector(A=@n2, B=@n2))
  n4 = TextureSampleParameter2D!(UVs=VertexInterpolator(VS=Divide(A=@n3, B=@n1)).PS, Bias, Apply View MipBias) {
   constMipValue=5
   group=TerrainColorMatch
   mipValueMode=TMVM_MipBias
   parameterName=GrassTextureToMatch
   samplerType=SAMPLERTYPE_LinearColor
   texture=/Figment_S06/FN_6_20/Athena/Environments/Landscape/Textures/T_Athena_Terrain_CombinedColors_01
  }
  n5 = Multiply(
   A=@n4.RGB
   B=CollectionParameter() {collection=/FigmentEnvironment/FN_6_20/Athena/Environments/Landscape/MPC/MPC_Landscape, parameterName=ColormapDivisor}
  )
  n6 = VectorParameter() {parameterName=NeutralGrassColor, defaultValue=<rgba 0.0675 0.145 0.02 1>}
  FunctionOutput(None=Multiply(A=@n5, B=@n6.RGB))
  FunctionOutput(None=@n6.RGB) {outputName=GrassColor}
  FunctionOutput(None=@n5) {outputName=GrassNoColor}
  Multiply(A=@n2)
  Multiply(A=@n2) {constB=-1}
  ScalarParameter() {defaultValue=50432, group=TerrainColorMatch, parameterName=TerrainSizeHalf}
 ]}
